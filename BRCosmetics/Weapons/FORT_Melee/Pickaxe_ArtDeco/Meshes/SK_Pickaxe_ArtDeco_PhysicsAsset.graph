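# Physics asset for the Art Deco pickaxe melee cosmetic's skeletal mesh, providing its physics collision shape.
# It holds a single body on the Melee_TwoHanded_Handle bone, made of one rotated, offset box.
# It defines no constraints, so the pickaxe is represented as one rigid piece with no articulated parts.
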PhysicsAsset /BRCosmetics/Weapons/FORT_Melee/Pickaxe_ArtDeco/Meshes/SK_Pickaxe_ArtDeco_PhysicsAsset
bodies {Melee_TwoHanded_Handle [{shapeType=Box, extentX=83.6255, extentY=13.5226, extentZ=143.4466, center=<2.253 -0.9082 22.6035>, rotation=<rot -0.0032 -90 0.6533>}]}
constraints []
